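# Low-detail water material for the scrying pool quest prop, with an IsCharacter switch for a water-figure variant.
# Two panning wave normals blend to a radial ripple over time, and sine waves offset the vertices.
# The character variant adds character colors, opacity and normals, blended with the pool's.
# GameTime and Fadeout animate a world-height band that drives bottom blending and a vertex push, apparently merging the figure into the water.
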Material /Game/Environments/Helios/Props/Quest/Quest_ScryingPool_A/Materials/M_ScryingPoolWater_LowDetail
graph {
 properties {
  bUseMaterialAttributes=true
  bUsedWithMeshDeformer=true
  bUsedWithNanite=true
  bUsedWithSkeletalMesh=true
  pixelDepthOffsetMode=PDOM_Legacy
  refractionMethod=RM_IndexOfRefraction
  usedShadingModels=Default Lit
 }
 nodes [
  n1 = Multiply(A=Subtract(A=Time(), B=ScalarParameter() {parameterName=GameTime})) {constB=0.5}
  n2 = Saturate(None=Power(Base=Saturate(None=@n1), Exp=ScalarParameter() {defaultValue=3, parameterName=FadeInAnimEXp}))
  n3 = LinearInterpolate(
   A=LinearInterpolate(Alpha=@n2) {constA=200, constB=0}
   B=LinearInterpolate(Alpha=@n2) {constB=200}
   Alpha=ScalarParameter() {parameterName=Fadeout}
  )
  n4 = Subtract(
   A=Subtract(A=WorldPosition() {worldPositionShaderOffset=WPT_ExcludeAllShaderOffsets}.Z, B=@n3)
   B=Add(
    A=ScalarParameter() {defaultValue=-5, parameterName=WordPosition_Z}
    B=ScalarParameter() {defaultValue=88, parameterName=WordPosition_Z_StaticOffset}
   )
  )
  n5 = Subtract(A=@n4, B=ScalarParameter() {defaultValue=-5, parameterName=WordPosition_Z_2nd})
  n6 = AppendVector(A=Constant2Vector(), B=LinearInterpolate(A=@n5, Alpha=Saturate(None=@n5)) {constB=0})
  n7 = AppendVector(A=Constant2Vector(), B=ScalarParameter() {parameterName=Sink})
  n8 = Normalize(VectorInput=Multiply(A=VertexNormalWS(), B=Constant3Vector() {constant=<rgba 1 1 0 0>}))
  n9 = Divide(A=@n4, B=ScalarParameter() {defaultValue=56.1126, parameterName=BottomBlendDistance})
  n10 = NamedRerouteDeclaration(None=Saturate(None=OneMinus(None=@n9))) {name=BottomBlend, nodeColor=<rgba 1 0 0.5625 1>}
  n11 = ScalarParameter() {defaultValue=15.7885, parameterName=BottomBlendFalloff}
  n12 = Saturate(None=LinearInterpolate(
    A=ScalarParameter() {defaultValue=-0.2, parameterName=BottomBlendFalloff_Low}
    B=ScalarParameter() {defaultValue=1.3, parameterName=BottomBlendFalloff_High}
    Alpha=Power(Base=@n10, Exp=@n11)
   ))
  n13 = Multiply(
   A=Multiply(A=Add(A=@n7, B=@n8), B=ScalarParameter() {defaultValue=33.4425, parameterName=Push})
   B=NamedRerouteDeclaration(None=@n12) {name=BottomBlendWIthContrast, nodeColor=<rgba 0 0.2734 1 1>}
  )
  Add(A=@n6, B=@n13)
  ComponentMask(None=MaterialFunctionCall() {materialFunction=&ComponentPivotLocation}.Result) {b=true}
  Constant3Vector() {constant=<rgba 0 0 1 0>}
  Frac(None=@n1)
  n14 = ScalarParameter() {defaultValue=0.03, parameterName=Speed}
  Multiply(A=@n14) {constB=-1}
  n15 = Constant2Vector()
  NamedRerouteDeclaration(None=Subtract(A=Subtract(A=@n13, B=@n6), B=AppendVector(A=@n15, B=@n3))) {name=WpoCharacterBlend, nodeColor=<rgba 0 0.6953 1 1>}
  n16 = Saturate(None=Fresnel(ExponentIn=ScalarParameter() {defaultValue=0.7117, parameterName=Fresnel}, Normal=VertexNormalWS()))
  NamedRerouteDeclaration(None=@n16) {name=Fresnel, nodeColor=<rgba 0.2656 1 0 1>}
  n17 = VectorParameter() {parameterName=ObjectPosition, defaultValue=<rgba 0 0 0 1>}
  n18 = ObjectPositionWS()
  n19 = ComponentMask(None=Subtract(A=WorldPosition().XYZ, B=StaticSwitchParameter(True=@n17.RGB, False=@n18) {parameterName=IsCharacter})) {g=true, r=true}
  n20 = Normalize(VectorInput=AppendVector(A=@n19, B=Constant()))
  n21 = MaterialFunctionCall(Vector or UVs=@n19) {materialFunction=&VectorToRadialValue}
  n22 = ScalarParameter() {defaultValue=1, parameterName=TilingRadial_X}
  n23 = ScalarParameter() {defaultValue=82.3486, parameterName=TilingRadial_Y}
  n24 = Panner(
   Coordinate=Divide(A=ComponentMask(None=@n21.Radial Coordinates) {g=true, r=true}, B=AppendVector(A=@n22, B=@n23))
   Speed=AppendVector(
    A=ScalarParameter() {defaultValue=-0.032, parameterName=Radial_Speed_X}
    B=ScalarParameter() {defaultValue=-0.248, parameterName=Radial_Speed_Y}
   )
  ) {speedX=0.01, speedY=0.01}
  n25 = TextureCoordinate()
  n26 = DDX(Value=@n25)
  n27 = DDY(Value=@n25)
  n28 = CrossProduct(
   A=@n20
   B=CrossProduct(
    A=TextureSample!(UVs=@n24, Tex, "DDX(UVs)"=@n26, "DDY(UVs)"=@n27, Apply View MipBias) {mipValueMode=TMVM_Derivative, samplerType=SAMPLERTYPE_Normal, texture=/FortWater/Textures/T_Water_TilingNormal_Waves_01}.RGB
    B=@n20
   )
  )
  Normalize(VectorInput=@n28)
  Reroute(None=@n12)
  Reroute(None=@n16)
  n29 = LinearInterpolate(
   A=VectorParameter() {parameterName=Color_Character, defaultValue=<rgba 0.0161 0.0421 0.0938 1>}.RGB
   B=VectorParameter() {parameterName=ColorCharacter2, defaultValue=<rgba 0.1707 0.5481 1 1>}.RGB
   Alpha=NamedRerouteUsage()
  )
  n30 = VectorParameter() {parameterName=Color, defaultValue=<rgba 0 0.0854 0.2031 1>}
  n31 = NamedRerouteUsage()
  n32 = ScalarParameter() {defaultValue=0.1712, parameterName=Specular_Character}
  n33 = ScalarParameter() {defaultValue=0.4, parameterName=Specular}
  n34 = NamedRerouteUsage()
  n35 = ScalarParameter() {parameterName=RoughnessCharacter}
  n36 = ScalarParameter() {parameterName=Roughness}
  n37 = NamedRerouteUsage()
  n38 = LinearInterpolate(
   A=ScalarParameter() {defaultValue=0.8116, parameterName=Opacity_Charcter}
   B=ScalarParameter() {parameterName=Opacity_Charcter2}
   Alpha=NamedRerouteUsage()
  )
  n39 = ScalarParameter() {parameterName=Opacity}
  n40 = NamedRerouteUsage()
  n41 = TextureSampleParameter2D() {
   parameterName=Normals
   samplerType=SAMPLERTYPE_Normal
   texture=/Game/Characters/Player/Male/Medium/Bodies/M_Med_Soldier_04/Textures/M_Med_Soldier_Body_01_N
  }
  n42 = PreSkinnedPosition()
  n43 = ScalarParameter() {defaultValue=-0.0423, parameterName=Speed2}
  n44 = AppendVector(A=@n43, B=@n43)
  n45 = AppendVector(A=@n14, B=@n14)
  n46 = MaterialFunctionCall(
   BaseNormal=@n41.RGB
   AdditionalNormal=MaterialFunctionCall!(
    TextureObject=TextureObject() {samplerType=SAMPLERTYPE_Normal, texture=&T_Water_TilingNormal_Waves_01}
    TextureSize=ScalarParameter() {defaultValue=86.3676, parameterName=UVScale_Normal_Character}
    Normal
    WorldSpace
    ProjectionTransitionContrast
    WorldPosition=VertexInterpolator(VS=@n42).PS
    Use High Quality Normals
    Pan_YZ=@n44
    Pan_XZ=@n45
    Pan_XY=@n45
   ) {materialFunction=/Game/Weapons/WeaponSkins/Wraps/MaterialFunctions/WorldAlignedNormal_ComplexPan}.XYZ Texture
  ) {materialFunction=&BlendAngleCorrectedNormals}
  n47 = Divide(A=WorldPosition().XY, B=ScalarParameter() {defaultValue=200, parameterName=UVScale})
  n48 = Panner(Coordinate=@n47, Speed=@n44)
  n49 = Panner(Coordinate=@n47, Speed=@n45)
  n50 = Time()
  n51 = ScalarParameter() {parameterName=GameTime_Ripple}
  n52 = Saturate(None=Multiply(A=Subtract(A=@n50, B=@n51)))
  n53 = LinearInterpolate(
   A=Add(
    A=TextureSample!(UVs=@n48, Tex, Apply View MipBias) {samplerType=SAMPLERTYPE_Normal, texture=&T_Water_TilingNormal_Waves_01}.RGB
    B=TextureSample(UVs=@n49) {samplerType=SAMPLERTYPE_Normal, texture=&T_Water_TilingNormal_Waves_01}.RGB
   )
   B=@n28
   Alpha=LinearInterpolate(A=LinearInterpolate(Alpha=@n52) {constA=1, constB=0}, B=LinearInterpolate(Alpha=@n52), Alpha=ScalarParameter() {parameterName=BlendToRadial})
  )
  n54 = Transform(None=Normalize(VectorInput=@n53)) {transformSourceType=TRANSFORMSOURCE_World, transformType=TRANSFORM_Tangent}
  n55 = NamedRerouteUsage()
  n56 = NamedRerouteUsage()
  n57 = Multiply(A=Time(), B=ScalarParameter() {defaultValue=0.2907, parameterName=WaveSpeed})
  n58 = WorldPosition()
  n59 = ScalarParameter() {defaultValue=85.055, parameterName=WaveFrequency}
  n60 = MaterialFunctionCall(Float2=Add(A=Frac(None=@n57), B=Divide(A=@n58.XY, B=@n59))) {materialFunction=&BreakOutFloat2Components}
  n61 = Multiply(A=Sine(None=@n60.R), B=Sine(None=@n60.G))
  n62 = ScalarParameter() {defaultValue=-0.5, parameterName=WPO_WaveHeight_Bias}
  n63 = AppendVector(A=Constant2Vector(), B=Multiply(A=Add(A=@n61, B=@n62), B=ScalarParameter() {defaultValue=1, parameterName=WPO_WaveHeight}))
  SetMaterialAttributes!(
   MaterialAttributes
   Base Color=StaticSwitchParameter(True=LinearInterpolate(A=@n29, B=@n30.RGB, Alpha=@n31), False=@n30.RGB) {parameterName=IsCharacter}
   Metallic
   Specular=StaticSwitchParameter(True=LinearInterpolate(A=@n32, B=@n33, Alpha=@n34), False=@n33) {parameterName=IsCharacter}
   Roughness=StaticSwitchParameter(True=LinearInterpolate(A=@n35, B=@n36, Alpha=@n37), False=@n36) {parameterName=IsCharacter}
   Anisotropy
   Emissive Color
   Opacity=StaticSwitchParameter(True=LinearInterpolate(A=@n38, B=@n39, Alpha=@n40), False=@n39) {parameterName=IsCharacter}
   Opacity Mask=MaterialFunctionCall(Alpha Threshold=ScalarParameter() {defaultValue=0.7276, parameterName=Dither}) {materialFunction=&DitherTemporalAA}.Result
   Normal=StaticSwitchParameter(True=LinearInterpolate(A=@n46.Result, B=@n54, Alpha=@n55), False=@n54) {parameterName=IsCharacter}
   World Position Offset=StaticSwitchParameter(True=Add(A=@n63, B=@n56), False=@n63) {parameterName=IsCharacter}
   Refraction=ScalarParameter() {defaultValue=1.01, parameterName=Refraction}
  ) {attributeSetTypes [
    #1
    #2
    #3
    #4
    #5
    #6
    #7
    #8
    #9
    #10
    #11
   ]}
  Sine(None=@n57)
  StaticSwitchParameter(True=@n28, False=@n53) {parameterName=IsCharacter}
 ]
}
